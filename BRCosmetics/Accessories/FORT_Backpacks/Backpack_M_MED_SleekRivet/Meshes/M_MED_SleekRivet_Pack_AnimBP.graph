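# Animation Blueprint for the SleekRivet backpack cosmetic, offsetting the pack's Body bone and making its four exhaust bones jitter procedurally.
# Each update mixes four fast real-time sine waves, differing in frequency, phase and amplitude, into four rotators.
# The AnimGraph chains additive Transform Bone nodes on the input pose: two fixed Body offsets, then those exhaust rotations.
# Rotate_01, remapped from spine yaw, is computed but unused in the pose.
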AnimBlueprint /BRCosmetics/Accessories/FORT_Backpacks/Backpack_M_MED_SleekRivet/Meshes/M_MED_SleekRivet_Pack_AnimBP
target_skeleton=&/M_MED_SleekRivet_Pack_Skeleton
graphs [
 AnimGraph {nodes [
   n1 = "Transform (Modify) Bone - Bone: Body"(ComponentPose=Local To Component(LocalPose=Input Pose()), Alpha=1, Translation="0,0.1,1", Rotation="0,0,-2", Scale="1,1,1") {
    rotationMode=BMM_Additive
    scaleMode=BMM_Additive
    translationMode=BMM_Additive
    alphaBoolBlend {blendOption=Linear}
    boneToModify {boneName=Body}
    rotation {roll=-2}
    translation {y=0.1, z=1}
   }
   n2 = "Transform (Modify) Bone - Bone: Body"(ComponentPose=@n1, Alpha=1, Translation="0,-7,4", Rotation="0,0,40", Scale="1,1,1") {
    rotationMode=BMM_Additive
    scaleMode=BMM_Additive
    translationMode=BMM_Additive
    alphaBoolBlend {blendOption=Linear}
    boneToModify {boneName=Body}
    rotation {roll=40}
    translation {y=-7, z=4}
   }
   n3 = "Transform (Modify) Bone - Bone: dyn_exhaust_01_r"(ComponentPose=@n2, Alpha=1, Translation="0,0,0", Rotation=Get Exhaust_01_R() -> (Exhaust_01_R:Rotator), Scale="1,1,1") {
    rotationMode=BMM_Additive
    scaleMode=BMM_Additive
    translationMode=BMM_Additive
    alphaBoolBlend {blendOption=Linear}
    boneToModify {boneName=dyn_exhaust_01_r}
   }
   n4 = "Transform (Modify) Bone - Bone: dyn_exhaust_02_r"(ComponentPose=@n3, Alpha=1, Translation="0,0,0", Rotation=Get Exhaust_02_R() -> (Exhaust_02_R:Rotator), Scale="1,1,1") {
    rotationMode=BMM_Additive
    scaleMode=BMM_Additive
    translationMode=BMM_Additive
    alphaBoolBlend {blendOption=Linear}
    boneToModify {boneName=dyn_exhaust_02_r}
   }
   n5 = "Transform (Modify) Bone - Bone: dyn_exhaust_01_l"(ComponentPose=@n4, Alpha=1, Translation="0,0,0", Rotation=Get Exhaust_01_L() -> (Exhaust_01_L:Rotator), Scale="1,1,1") {
    rotationMode=BMM_Additive
    scaleMode=BMM_Additive
    translationMode=BMM_Additive
    alphaBoolBlend {blendOption=Linear}
    boneToModify {boneName=dyn_exhaust_01_l}
   }
   n6 = "Transform (Modify) Bone - Bone: dyn_exhaust_02_l"(ComponentPose=@n5, Alpha=1, Translation="0,0,0", Rotation=Get Exhaust_02_L() -> (Exhaust_02_L:Rotator), Scale="1,1,1") {
    rotationMode=BMM_Additive
    scaleMode=BMM_Additive
    translationMode=BMM_Additive
    alphaBoolBlend {blendOption=Linear}
    boneToModify {boneName=dyn_exhaust_02_l}
   }
   Output Pose(Result=Component To Local(ComponentPose=@n6)) {layerGroup=None}
  ]}
 EventGraph {nodes [
   Event BlueprintUpdateAnimation()
   > n1 = Sequence()
   n2 = GetDeltaTransformFromRefPose!(self=GetOwnerSkeletalMeshComponent(), BoneName=spine_05, BaseName=spine_01) -> (ReturnValue_Location:Vector, ReturnValue_Rotation:Rotator, ReturnValue_Scale:Vector)
   > Set Rotate_01(Rotate_01:double=MapRangeClamped(Value=BreakRotator(InRot=@n2.ReturnValue_Rotation).Yaw, InRangeA=-25, InRangeB=-65)) -> (Output_Get:double)
   n3 = GetRealTimeSeconds()
   n4 = "Sin (Radians)"(A=float + float(A=float * float(A=@n3, B=75) {is=title K2Node_PromotableOperator}, B=5))
   n5 = Set Exhaust_Anim_01(execute=@n1.then_1, Exhaust_Anim_01:double=float * float(A=float + float(A=@n4, B=1), B=0.6)) -> (Output_Get:double)
   n6 = "Sin (Radians)"(A=float + float(A=float * float(A=@n3, B=50), B=4))
   > n7 = Set Exhaust_Anim_02(Exhaust_Anim_02:double=float * float(A=float + float(A=@n6, B=1), B=0.8)) -> (Output_Get:double)
   n8 = "Sin (Radians)"(A=float + float(A=float * float(A=@n3, B=80), B=3))
   > n9 = Set Exhaust_Anim_03(Exhaust_Anim_03:double=float * float(A=float + float(A=@n8, B=1), B=0.4)) -> (Output_Get:double)
   n10 = "Sin (Radians)"(A=float + float(A=float * float(A=@n3, B=45), B=2))
   > n11 = Set Exhaust_Anim_04(Exhaust_Anim_04:double=float * float(A=float + float(A=@n10, B=1), B=0.7)) -> (Output_Get:double)
   > Set Exhaust_01_R(Exhaust_01_R:Rotator=MakeRotator(Roll=@n5, Pitch=@n7)) -> (Output_Get:Rotator)
   > Set Exhaust_02_R(Exhaust_02_R:Rotator=MakeRotator(Roll=@n9, Pitch=@n11)) -> (Output_Get:Rotator)
   > Set Exhaust_01_L(Exhaust_01_L:Rotator=MakeRotator(Roll=@n5, Pitch=@n11)) -> (Output_Get:Rotator)
   > Set Exhaust_02_L(Exhaust_02_L:Rotator=MakeRotator(Roll=@n9, Pitch=@n7)) -> (Output_Get:Rotator)
   TryGetPawnOwner()
  ]}
]
variables [
 Exhaust_01_L {type=Rotator}
 Exhaust_01_R {type=Rotator}
 Exhaust_02_L {type=Rotator}
 Exhaust_02_R {type=Rotator}
 Exhaust_Anim_01 {type=number}
 Exhaust_Anim_02 {type=number}
 Exhaust_Anim_03 {type=number}
 Exhaust_Anim_04 {type=number}
 Rotate_01 {type=number}
]
